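# Debug Niagara system for the CarbideWeld emote FX, with all four emitters disabled.
# NormalLine, PosBug and PosBugSeperateSysm burst sprites that sample two socket rotations, lerp between them by a camera-vector factor, and align to the resulting quaternion.
# NE_Debug_Float displays a camera-derived blend factor as sprite color.
# The PosBug variants test a fixed positional offset, either as a separate CirclePos attribute with a second renderer or applied directly to particle position.
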NiagaraSystem /BRCosmetics/Animation/Game/MainPlayer/Emotes/CarbideWeld/FX/NS_CARBIDEDEBUG
dependencies {
 usedDataInterfaces [
  &NiagaraDataInterfaceColorCurve
  &NiagaraDataInterfaceCamera
  &NiagaraDataInterfaceVectorCurve
  &NiagaraDataInterfaceSocketReader
 ]
 usedDynamicInputs [/Niagara/DynamicInputs/Multiply/Multiply_Vector4_ByFloat]
 usedModules [
  &SystemState
  &self:SystemSpawnScript.SystemScriptSource.SystemScriptGraph.NiagaraNodeAssignment_4.SetVariables_#1
 ]
}
emitters {
 NE_Debug_Float {
  bEnabled=false
  emitterSpawnScript [/SetVariables(Emitter.Digits=2, Emitter.DisplayScalar=true) {isSetParameters=true}]
  emitterUpdateScript [
   &EmitterState()
   &SpawnBurst_Instantaneous(Use Spawn Probability=false, UseLoopCountLimit=false)
  ]
  particleSpawnScript [&InitializeParticle(
    Color Mode=::NewEnumerator0
    Lifetime=$Emitter.CurrentLoopDuration
    Position Offset=<0 0 300>
    Sprite Size Mode=::NewEnumerator3
    Uniform Sprite Size=125
    UsePositionOffset=true
    Write Material Random=true
   )]
  particleUpdateScript [
   &ParticleState(Kill Particles When Lifetime Has Elapsed=false)
   &Color(Color=/Niagara/DynamicInputs/TypeConversions/MakeLinearColorFromFloat(Value:NiagaraFloat=&RemapRange(
      Input Value To Remap=&Abs_Float(Float:NiagaraFloat=&DotProduct(
        Vector A=&VectorFromCurve(
         VectorCurve=DataInterface() {propertyValues {XCurve {keys [<at 0 1>,<at 0.6 0>]}, YCurve {keys [<at 0 1>,<at 0.6 0>]}, ZCurve {keys [<at 0 1>,<at 0.6 0>]}}}
         CurveIndex=&Abs_Float(Float:NiagaraFloat=&MakeFloatFromVector(
           Vector=&ReturnCameraVector(Camera Vectors=::NewEnumerator1, Coordinate Space=::Simulation)
           Channel=::NewEnumerator0
          ))
         Scale Curve=<1 0 0>
        )
        Vector B=<1 0 0>
       ))
      Input Min=0.2
      Input Max=0.9
      Output Min=0
      Output Max=1
     )))
  ]
  rendererClasses [&NiagaraSpriteRendererProperties]
  renderers [{rendererClass=&NiagaraSpriteRendererProperties}]
 }
 NormalLine {
  bEnabled=false
  emitterUpdateScript [
   &EmitterState()
   &SpawnRate(SpawnGroup=1, Use Spawn Probability=false) {enabled=false}
   &SpawnBurst_Instantaneous(Use Spawn Probability=false, UseLoopCountLimit=false)
  ]
  particleSpawnScript [&InitializeParticle(
    Color=$System.LineColor
    Lifetime=10
    Position Offset=<0 8.9801 1.8622>
    Position Offset Coordinate Space=::Local
    Sprite Rotation Angle=90
    Sprite Rotation Mode=::NewEnumerator0
    Sprite Size Mode=::NewEnumerator3
    Uniform Sprite Size=36
    UsePositionOffset=true
    Write Material Random=true
   )]
  particleUpdateScript [
   &SocketLocation(
    SamplePosOffset=false
    Sampled Orientation Coordinate Space:ENiagaraCoordinateSpace=::World
    Socket Reader=$System.Socket Reader
    Socket Sampling Mode=::NewEnumerator3
    VelocitySpeedLimit=false
   )
   &SocketLocation(
    SamplePosOffset=false
    Sampled Orientation Coordinate Space:ENiagaraCoordinateSpace=::World
    Socket Index=1
    Socket Reader=$System.Socket Reader
    Socket Sampling Mode=::NewEnumerator3
    VelocitySpeedLimit=false
   ) {module=SocketLocation001}
   &ParticleState()
   &ScaleColor(
    Linear Color Curve=DataInterface() {propertyValues {
      AlphaCurve {keys [<at 0.1588 1>]}
      BlueCurve {keys [<at 0 1>,<at 1 1>]}
      GreenCurve {keys [<at 0 1>,<at 1 1>]}
      RedCurve {keys [<at 0 1>,<at 1 1>]}
     }}
    Scale Mode=::NewEnumerator2
   ) {enabled=false}
   &SolveForcesAndVelocity(Clamp Velocity=false, Limit Acceleration=false)
   /SetVariables(Particles.Quat:Quat4f=&LerpQuaternion(
     Quaternion A:Quat4f=$Output.SocketLocation001.SampledRotation
     Quaternion B:Quat4f=$Output.SocketLocation.SampledRotation
     Lerp Factor:NiagaraFloat=&RemapRange(
      Input Value To Remap=&Abs_Float(Float:NiagaraFloat=&DotProduct(
        Vector A=&VectorFromCurve(
         VectorCurve=DataInterface() {propertyValues {XCurve {keys [<at 0 1>,<at 0.6 0>]}, YCurve {keys [<at 0 1>,<at 0.6 0>]}, ZCurve {keys [<at 0 1>,<at 0.6 0>]}}}
         CurveIndex=&MakeFloatFromVector(
          Vector=&ReturnCameraVector(Camera Vectors=::NewEnumerator1, Coordinate Space=::Simulation)
          Channel=::NewEnumerator0
         )
         Scale Curve=<1 0 0>
        )
        Vector B=<1 0 0>
       ))
      Input Min=0.2
      Input Max=0.9
      Output Min=0
      Output Max=1
     )
     Enable Spherical Linear Interpolation:NiagaraBool=false
    )) {isSetParameters=true, module=SetVariables_#2}
   &AlignSpriteToMeshOrientation(Orientation Quaternion:Quat4f=$Particles.Quat, SetAlignment=true, SetFacing=true)
  ]
  rendererClasses [&NiagaraSpriteRendererProperties]
  renderers [{rendererClass=&NiagaraSpriteRendererProperties}]
 }
 PosBug {
  bEnabled=false
  emitterUpdateScript [
   &EmitterState()
   &SpawnRate(SpawnGroup=1, Use Spawn Probability=false) {enabled=false}
   &SpawnBurst_Instantaneous(Use Spawn Probability=false, UseLoopCountLimit=false)
  ]
  particleSpawnScript [&InitializeParticle(
    Color=$System.LineColor
    Lifetime=10
    Position Offset=<0 8.9801 1.8622>
    Position Offset Coordinate Space=::Local
    Sprite Rotation Angle=90
    Sprite Rotation Mode=::NewEnumerator0
    Sprite Size Mode=::NewEnumerator3
    Uniform Sprite Size=36
    UsePositionOffset=true
    Write Material Random=true
   )]
  particleUpdateScript [
   &SocketLocation(
    SamplePosOffset=false
    Sampled Orientation Coordinate Space:ENiagaraCoordinateSpace=::World
    Socket Reader=$System.Socket Reader
    Socket Sampling Mode=::NewEnumerator3
    VelocitySpeedLimit=false
   )
   &SocketLocation(
    SamplePosOffset=false
    Sampled Orientation Coordinate Space:ENiagaraCoordinateSpace=::World
    Socket Index=1
    Socket Reader=$System.Socket Reader
    Socket Sampling Mode=::NewEnumerator3
    VelocitySpeedLimit=false
   ) {module=SocketLocation001}
   &ParticleState()
   /SetVariables(Particles.CirclePos:NiagaraPosition=&AddVectorToPosition(Position=$Particles.Position, Vector=<0 33.9961 -15.7091>)) {isSetParameters=true}
   &ScaleColor(
    Linear Color Curve=DataInterface() {propertyValues {
      AlphaCurve {keys [<at 0.1588 1>]}
      BlueCurve {keys [<at 0 1>,<at 1 1>]}
      GreenCurve {keys [<at 0 1>,<at 1 1>]}
      RedCurve {keys [<at 0 1>,<at 1 1>]}
     }}
    Scale Mode=::NewEnumerator2
   ) {enabled=false}
   &SolveForcesAndVelocity(Clamp Velocity=false, Limit Acceleration=false)
   /SetVariables(Particles.Quat:Quat4f=&LerpQuaternion(
     Quaternion A:Quat4f=$Output.SocketLocation001.SampledRotation
     Quaternion B:Quat4f=$Output.SocketLocation.SampledRotation
     Lerp Factor:NiagaraFloat=&ClampFloat(Float=&Abs_Float(Float:NiagaraFloat=&RemapRange(
        Input Value To Remap=&DotProduct(Vector A=&ReturnCameraVector(Camera Vectors=::NewEnumerator1, Coordinate Space=::Local), Vector B=<1 0 0>)
        Input Min=0
        Input Max=1
        Output Min=0
        Output Max=1
       )))
     Enable Spherical Linear Interpolation:NiagaraBool=false
    )) {isSetParameters=true, module=SetVariables_#2}
   &AlignSpriteToMeshOrientation(Orientation Quaternion:Quat4f=$Particles.Quat, SetAlignment=true, SetFacing=true)
   /SetVariables() {isSetParameters=true}
  ]
  rendererClasses [&NiagaraSpriteRendererProperties]
  renderers [
   {rendererClass=&NiagaraSpriteRendererProperties}
   {rendererClass=&NiagaraSpriteRendererProperties, rendererIndex=1}
  ]
 }
 PosBugSeperateSysm {
  bEnabled=false
  emitterUpdateScript [
   &EmitterState()
   &SpawnRate(SpawnGroup=1, Use Spawn Probability=false) {enabled=false}
   &SpawnBurst_Instantaneous(Use Spawn Probability=false, UseLoopCountLimit=false)
  ]
  particleSpawnScript [&InitializeParticle(
    Color=$System.LineColor
    Lifetime=10
    Position Mode=::NewEnumerator1
    Position Offset Coordinate Space=::Local
    Sprite Rotation Mode=::NewEnumerator0
    Sprite Size Mode=::NewEnumerator3
    Uniform Sprite Size=37.6938
    UsePositionOffset=true
    Write Material Random=true
   )]
  particleUpdateScript [
   &SocketLocation(
    SamplePosOffset=false
    Sampled Orientation Coordinate Space:ENiagaraCoordinateSpace=::World
    Socket Reader=$System.Socket Reader
    Socket Sampling Mode=::NewEnumerator3
    VelocitySpeedLimit=false
   )
   &SocketLocation(
    SamplePosOffset=false
    Sampled Orientation Coordinate Space:ENiagaraCoordinateSpace=::World
    Socket Index=1
    Socket Reader=$System.Socket Reader
    Socket Sampling Mode=::NewEnumerator3
    VelocitySpeedLimit=false
   ) {module=SocketLocation001}
   /SetVariables(Particles.Position=&AddVectorToPosition(Position=$Particles.Position, Vector=<0 33.9961 -15.7091>)) {isSetParameters=true}
   &ParticleState()
   &ScaleColor(
    Linear Color Curve=DataInterface() {propertyValues {
      AlphaCurve {keys [<at 0.1588 1>]}
      BlueCurve {keys [<at 0 1>,<at 1 1>]}
      GreenCurve {keys [<at 0 1>,<at 1 1>]}
      RedCurve {keys [<at 0 1>,<at 1 1>]}
     }}
    Scale Mode=::NewEnumerator2
   ) {enabled=false}
   &SolveForcesAndVelocity(Clamp Velocity=false, Limit Acceleration=false)
   /SetVariables(Particles.Quat:Quat4f=&LerpQuaternion(
     Quaternion A:Quat4f=$Output.SocketLocation001.SampledRotation
     Quaternion B:Quat4f=$Output.SocketLocation.SampledRotation
     Lerp Factor:NiagaraFloat=&ClampFloat(Float=&Abs_Float(Float:NiagaraFloat=&RemapRange(
        Input Value To Remap=&DotProduct(Vector A=&ReturnCameraVector(Camera Vectors=::NewEnumerator1, Coordinate Space=::Local), Vector B=<1 0 0>)
        Input Min=0
        Input Max=1
        Output Min=0
        Output Max=1
       )))
     Enable Spherical Linear Interpolation:NiagaraBool=false
    )) {isSetParameters=true, module=SetVariables_#2}
   &AlignSpriteToMeshOrientation(Orientation Quaternion:Quat4f=$Particles.Quat, SetAlignment=true, SetFacing=true)
  ]
  rendererClasses [&NiagaraSpriteRendererProperties]
  renderers [{rendererClass=&NiagaraSpriteRendererProperties}]
 }
}
summary {
 systemName=NS_CARBIDEDEBUG
 emitters [
  {bEnabled=false, emitterName=PosBug, rendererClasses [&NiagaraSpriteRendererProperties]}
  {bEnabled=false, emitterName=NE_Debug_Float, rendererClasses [&NiagaraSpriteRendererProperties]}
  {bEnabled=false, emitterName=PosBugSeperateSysm, rendererClasses [&NiagaraSpriteRendererProperties]}
  {bEnabled=false, emitterName=NormalLine, rendererClasses [&NiagaraSpriteRendererProperties]}
 ]
 userVariables [
  User.NewVector3f {type=Vector3f, description="", defaultValue {struct=Vector3f, value=<0 0 0>}}
  User.NormalizedNotifyProgress {type=NiagaraFloat, defaultValue=0, description=""}
  User.RibbonIDOffset {type=NiagaraInt32}
 ]
}
