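# Niagara system for the floor decal and accompanying effects of the CarbideWeld emote, judging by its name and emitters.
# Eight emitters, named as lightweight (LWE), cover two decals, a floor ring, two dotted rings, a trailing cylinder and two lights.
# System spawn uses a Set Variables assignment, and the system uses the FN_FrontEndFloatModifier dynamic input alongside front-end and front-end-preview user bools, suggesting values adjusted for menu display.
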NiagaraSystem /BRCosmetics/Animation/Game/MainPlayer/Emotes/CarbideWeld/FX/NS_Emote_CarbideWeld_Decal
dependencies {
 usedDataInterfaces []
 usedDynamicInputs [&FN_FrontEndFloatModifier]
 usedModules [
  &self:SystemSpawnScript.SystemScriptSource.SystemScriptGraph.NiagaraNodeAssignment_0.SetVariables_#1
  &SystemState
 ]
}
emitters {
 DecalLWE {rendererClasses [], renderers []}
 DecalLWE001 {rendererClasses [], renderers []}
 DottedRingBigLWE {rendererClasses [], renderers []}
 DottedRingSmallLWE {rendererClasses [], renderers []}
 LightLWE1 {rendererClasses [], renderers []}
 LightLWE1001 {rendererClasses [], renderers []}
 RingFloorLWE {rendererClasses [], renderers []}
 TrailingCylinderLWE {rendererClasses [], renderers []}
}
summary {
 systemName=NS_Emote_CarbideWeld_Decal
 emitters [
  {emitterName=DecalLWE, rendererClasses []}
  {emitterName=RingFloorLWE, rendererClasses []}
  {emitterName=DottedRingSmallLWE, rendererClasses []}
  {emitterName=DottedRingBigLWE, rendererClasses []}
  {emitterName=TrailingCylinderLWE, rendererClasses []}
  {emitterName=DecalLWE001, rendererClasses []}
  {emitterName=LightLWE1, rendererClasses []}
  {emitterName=LightLWE1001, rendererClasses []}
 ]
 userVariables [
  User.bIsFrontEnd {type=NiagaraBool}
  User.bIsFrontEndPreview {type=NiagaraBool}
 ]
}
