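# Niagara system for ring text sprites in the CarbideWeld emote, named RETIRED and with its only emitter, Line, disabled.
# Four timed bursts tag particles by spawn group and partition them across four local offsets.
# Opposite facing and alignment vectors, plus separate blue text and orange line attributes, accompany four sprite renderers.
# Sprites scale open and closed horizontally, with alpha driven by camera-vector dot product to fade at edge-on views.
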NiagaraSystem /BRCosmetics/Animation/Game/MainPlayer/Emotes/CarbideWeld/FX/NS_Emote_CarbideWeld_RingTextRETIRED
dependencies {
 usedDataInterfaces [
  &NiagaraDataInterfaceArrayFloat3
  &NiagaraDataInterfaceCurve
  &NiagaraDataInterfaceVector2DCurve
  &NiagaraDataInterfaceVectorCurve
  &NiagaraDataInterfaceCamera
 ]
 usedDynamicInputs [&VectorFromCurve, &MakeFloatFromVector]
 usedModules [
  &SystemState
  &self:SystemSpawnScript.SystemScriptSource.SystemScriptGraph.NiagaraNodeAssignment_0.SetVariables_#1
 ]
}
emitters {Line {
  bEnabled=false
  emitterUpdateScript [
   &EmitterState()
   &SpawnBurst_Instantaneous(Spawn Time=0.8, Use Spawn Probability=false, UseLoopCountLimit=false)
   &SpawnBurst_Instantaneous(Spawn Group=1, Spawn Time=1.6, Use Spawn Probability=false, UseLoopCountLimit=false) {module=SpawnBurst_Instantaneous001}
   &SpawnBurst_Instantaneous(Spawn Group=2, Spawn Time=0.8, Use Spawn Probability=false, UseLoopCountLimit=false) {module=SpawnBurst_Instantaneous002}
   &SpawnBurst_Instantaneous(Spawn Group=3, Spawn Time=1.6, Use Spawn Probability=false, UseLoopCountLimit=false) {module=SpawnBurst_Instantaneous003}
  ]
  particleSpawnScript [
   /SetVariables(Particles.VisibilityTag=&SpawnGroup()) {isSetParameters=true}
   &PartitionParticles(Number Of Partitions=4)
   &InitializeParticle(
    Lifetime=1.5
    Position Offset=&SelectVectorFromArray(
     Direct Array Index=$Particles.PartitionParticles.Partition
     Vector Selection Array=DataInterface() {propertyValues {FloatData [<0 -35 45>, <10 -75 53>, <0 35 100>, <6 55 109>]}}
    )
    Position Offset Coordinate Space=::Local
    Sprite Rotation Angle=90
    Sprite Rotation Mode=::NewEnumerator0
    Sprite Size Mode=::NewEnumerator3
    Uniform Sprite Size=53
    UsePositionOffset=true
    Write Material Random=true
   )
   &SpriteFacingAndAlignment(Sprite Alignment=<0 -1 0>, Sprite Facing=<1 0.3 0>, WriteAlignment=true, WriteFacing=true)
   /SetVariables(Particles.Alignment2:Vector3f=<0 1 0>, Particles.Facing2:Vector3f=<-1 -0.3 0>) {isSetParameters=true}
   &InheritVelocity(Speed Limit=false, UseSpeedThreshold=false) {enabled=false}
   /SetVariables(Particles.Alignment3:Vector3f=<0 -1 0>, Particles.Facing3:Vector3f=<1 -0.3 0>) {isSetParameters=true}
  ]
  particleUpdateScript [
   &ParticleState()
   &ScaleSpriteSize(
    Non-Uniform Curve Sprite Scale=DataInterface() {propertyValues {
      XCurve {keys [{tangentMode=RCTM_User, time=0, value=0}, {tangentMode=RCTM_User, time=0.0731, value=1}, <at 0.9418 1>, <at 1 0>]}
      YCurve {keys [{tangentMode=RCTM_User, time=0, value=1}]}
     }}
    Scale Sprite Size Mode=::NewEnumerator3
   )
   &SolveForcesAndVelocity(Clamp Velocity=false, Limit Acceleration=false)
   /SetVariables(
    Particles.RotationText=180
    Particles.TextColor:LinearColor=&MakeLinearColorFromVectorAndFloat(
     "Vector (RGB)"=&MakeVectorFromLinearColorRGB(Color=<rgba 0 0.3898 1.2 1>)
     "Float (Alpha)"=&RemapRange(
      Input Value To Remap=&Abs_Float(Float:NiagaraFloat=&DotProduct(Vector A=&ReturnCameraVector(Camera Vectors=::NewEnumerator1, Coordinate Space=::Local), Vector B=<1 0 0>))
      Input Min=0.2
      Input Max=0.9
      Output Min=0
      Output Max=1
     )
    )
    Particles.TextSize:Vector2f=/Niagara/DynamicInputs/ScaleByCurve/ScaleVector2DByCurve(
     InitialValue:Vector2f=<14.6124 32.8192>
     ScaleCurve:NiagaraDataInterfaceVectorCurve=DataInterface() {propertyValues {
       XCurve {defaultValue=1, keys [{tangentMode=RCTM_User, time=0, value=0}, <at 0.05 1>, <at 0.93 1>, {tangentMode=RCTM_User, time=1, value=0}]}
       YCurve {defaultValue=1, keys [{tangentMode=RCTM_User, time=0, value=1}, {tangentMode=RCTM_User, time=0.2937, value=1}, <at 0.75 1>, <at 1 1>]}
       ZCurve {keys []}
      }}
     CurveIndex:NiagaraFloat=$Particles.NormalizedAge
    )
    Particles.float:NiagaraFloat=&FloatFromCurve(FloatCurve=DataInterface() {propertyValues {Curve {keys [<at 0 0>,<at 0.8 3>]}}})
   ) {isSetParameters=true}
   /SetVariables(Particles.LineColor:LinearColor=&MakeLinearColorFromVectorAndFloat(
     "Vector (RGB)"=&MakeVectorFromLinearColorRGB(Color=<rgba 2 0.3811 0.0324 1>)
     "Float (Alpha)"=&Multiply_Float(
      A=&FloatFromCurve(FloatCurve=DataInterface() {propertyValues {Curve {keys [<at 0.2 1>,<at 2.8503 1>]}}}, CurveIndex=$Engine.System.Age)
      B=&RemapRange(
       Input Value To Remap=&Abs_Float(Float:NiagaraFloat=&DotProduct(Vector A=&ReturnCameraVector(Camera Vectors=::NewEnumerator1, Coordinate Space=::Local), Vector B=<1 0 0>))
       Input Min=0.2
       Input Max=0.9
       Output Min=0
       Output Max=1
      )
     )
    )) {isSetParameters=true}
  ]
  rendererClasses [&NiagaraSpriteRendererProperties]
  renderers [
   {rendererClass=&NiagaraSpriteRendererProperties}
   {rendererClass=&NiagaraSpriteRendererProperties, rendererIndex=1}
   {rendererClass=&NiagaraSpriteRendererProperties, rendererIndex=2}
   {rendererClass=&NiagaraSpriteRendererProperties, rendererIndex=3}
  ]
 }}
summary {
 systemName=NS_Emote_CarbideWeld_RingTextRETIRED
 emitters [{bEnabled=false, emitterName=Line, rendererClasses [&NiagaraSpriteRendererProperties]}]
 userVariables []
}
