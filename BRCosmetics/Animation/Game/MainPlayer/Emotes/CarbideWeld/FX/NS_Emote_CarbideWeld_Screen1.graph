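# Niagara system for the first screen effect of the CarbideWeld emote, built from animated circle and text sprites plus six LWE emitters with no detail shown.
# StaticCircle fires staggered bursts, setting position and lifetime by partition and a visibility tag by spawn group for four sprite renderers.
# Curves animate scale, blinking text and typing SubUV.
# Alpha fades with camera angle, and front-end modifiers switch positions and fade curves between gameplay and menus.
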NiagaraSystem /BRCosmetics/Animation/Game/MainPlayer/Emotes/CarbideWeld/FX/NS_Emote_CarbideWeld_Screen1
dependencies {
 usedDataInterfaces [
  &NiagaraDataInterfaceArrayFloat
  &NiagaraDataInterfaceArrayFloat3
  &NiagaraDataInterfaceCurve
  &NiagaraDataInterfaceVectorCurve
  &NiagaraDataInterfaceCamera
  &NiagaraDataInterfaceColorCurve
 ]
 usedDynamicInputs [&Lerp_Vector]
 usedModules [
  &self:SystemSpawnScript.SystemScriptSource.SystemScriptGraph.NiagaraNodeAssignment_2.SetVariables_#1
  &SystemState
  &self:SystemSpawnScript.SystemScriptSource.SystemScriptGraph.NiagaraNodeAssignment_0.SetVariables_#2
  &self:SystemSpawnScript.SystemScriptSource.SystemScriptGraph.NiagaraNodeAssignment_1.SetVariables_#3
 ]
}
emitters {
 BigBarLWE {rendererClasses [], renderers []}
 FlashLWE {rendererClasses [], renderers []}
 LineLWE {rendererClasses [], renderers []}
 PentagonLWE {rendererClasses [], renderers []}
 SmallBarsLWE {rendererClasses [], renderers []}
 SpinningCircleLWE {rendererClasses [], renderers []}
 StaticCircle {
  emitterUpdateScript [
   &EmitterState()
   &SpawnBurst_Instantaneous(Spawn Group=1, Use Spawn Probability=false, UseLoopCountLimit=false) {module=SpawnBurst_Instantaneous001}
   &SpawnBurst_Instantaneous(Use Spawn Probability=false, UseLoopCountLimit=false)
   &SpawnBurst_Instantaneous(Spawn Group=2, Spawn Time=1.5, Use Spawn Probability=false, UseLoopCountLimit=false) {module=SpawnBurst_Instantaneous002}
   &SpawnBurst_Instantaneous(Spawn Group=3, Spawn Time=1.3, Use Spawn Probability=false, UseLoopCountLimit=false) {module=SpawnBurst_Instantaneous003}
  ]
  particleSpawnScript [
   &PartitionParticles(Number Of Partitions=5)
   &InitializeParticle(
    Color=$System.OrangeColors
    Lifetime=&SelectFloatFromArray(
     Direct Array Index=$Particles.PartitionParticles.Partition
     Float Selection Array=DataInterface() {propertyValues {FloatData [3.3,1.8,1.5,1.4]}}
    )
    Position=&FN_FrontEndVector3Modifier(
     Gameplay=&SelectVectorFromArray(
      Direct Array Index=$Particles.PartitionParticles.Partition
      Vector Selection Array=DataInterface() {propertyValues {FloatData [<0 -6.6238 -5.1064>, <3 -5.197 11.2302>, <0 8.5 3.5>]}}
     )
     Front End=&AddVectorToPosition(
      Position=&SelectVectorFromArray(
       Direct Array Index=$Particles.PartitionParticles.Partition
       Vector Selection Array=DataInterface() {propertyValues {FloatData [<0 -6.6238 -5.1064>, <0 -5.197 11.2302>, <0 8.5 3.5>]}}
      )
      Vector=$User.FrontendPos
     )
    )
    Position Mode=::NewEnumerator1
    Sprite Rotation Mode=::NewEnumerator0
    Sprite Size Mode=::NewEnumerator3
    Uniform Sprite Size=33.8249
    UsePositionOffset=false
    Write Material Random=true
   )
   &SpriteFacingAndAlignment(
    Sprite Facing=&AddVectorToPosition(Position=<1 0 0>, Vector=<0 0 0>)
    WriteAlignment=true
    WriteFacing=true
   )
   /SetVariables(
    Particles.Alignment:Vector3f=<1 0 0>
    Particles.Facing:Vector3f=<-40 -1 0>
    Particles.Position2:NiagaraPosition=&FN_FrontEndVector3Modifier(Gameplay=<0 10 7>, Front End=&AddVectorToPosition(Position=<0 10 7>, Vector=$User.FrontendPos))
    Particles.VisibilityTag=&SpawnGroup()
   ) {isSetParameters=true}
   /SetVariables(Particles.StaticCirclePos:NiagaraPosition=&FN_FrontEndVector3Modifier(
     Gameplay=<0 -6.6238 -5.1064>
     Front End=&AddVectorToPosition(Position=<0 -6.6238 -5.1064>, Vector=$User.FrontendPos)
    )) {isSetParameters=true}
  ]
  particleUpdateScript [
   &ParticleState()
   &ScaleSpriteSize(Uniform Curve Sprite Scale=DataInterface() {propertyValues {Curve {keys [
        {arriveTangent=34.6101, interpMode=RCIM_Cubic, leaveTangent=34.6101, tangentMode=RCTM_User, time=0, value=-0.025}
        {arriveTangent=1.3341, time=0.0925, value=1}
        {arriveTangent=0.0751, interpMode=RCIM_Cubic, leaveTangent=0.0751, tangentMode=RCTM_User, time=0.8963, value=1}
        {arriveTangent=-8.0338, interpMode=RCIM_Cubic, leaveTangent=-8.0338, tangentMode=RCTM_User, time=1, value=0}
       ]}}})
   &ScaleColor(
    Scale Alpha=&RemapRange(
     Input Value To Remap=&Abs_Float(Float:NiagaraFloat=&DotProduct(
       Vector A=&FN_FrontEndVector3Modifier(
        Gameplay=&VectorFromCurve(
         VectorCurve=DataInterface() {propertyValues {XCurve {keys [<at 0 0>,<at 0.2 1>]}, YCurve {keys [<at 0.7 1>,<at 1 0>]}, ZCurve {keys [<at 0.7 1>,<at 1 0>]}}}
         CurveIndex=&Abs_Float(Float:NiagaraFloat=&MakeFloatFromVector(
           Vector=&ReturnCameraVector(Camera Vectors=::NewEnumerator1, Coordinate Space=::Local)
           Channel=::NewEnumerator0
          ))
         Scale Curve=<1 0 0>
        )
        Front End=&VectorFromCurve(
         VectorCurve=DataInterface() {propertyValues {XCurve {keys [<at 0 0>,<at 0.5 1>]}, YCurve {keys [<at 0.7 1>,<at 1 0>]}, ZCurve {keys [<at 0.7 1>,<at 1 0>]}}}
         CurveIndex=&Abs_Float(Float:NiagaraFloat=&MakeFloatFromVector(
           Vector=&ReturnCameraVector(Camera Vectors=::NewEnumerator1, Coordinate Space=::Local)
           Channel=::NewEnumerator0
          ))
         Scale Curve=<1 0 0>
        )
       )
       Vector B=<1 0 0>
      ))
     Input Min=0.2
     Input Max=0.9
     Output Min=0
     Output Max=1
    )
    Scale RGB=&MakeVectorFromLinearColorRGB(Color=<rgba 2 0.3811 0.0324 1>)
    ScaleA=true
    ScaleRGB=true
   )
   &SolveForcesAndVelocity(Clamp Velocity=false, Limit Acceleration=false)
   /SetVariables(
    Particles.Text Color:LinearColor=&MakeLinearColor(
     R:NiagaraFloat=0
     G:NiagaraFloat=0.136
     B:NiagaraFloat=1
     A:NiagaraFloat=&RemapRange(
      Input Value To Remap=&Abs_Float(Float:NiagaraFloat=&DotProduct(
        Vector A=&FN_FrontEndVector3Modifier(
         Gameplay=&VectorFromCurve(
          VectorCurve=DataInterface() {propertyValues {XCurve {keys [<at 0 0>,<at 0.2 1>]}, YCurve {keys [<at 0.7 1>,<at 1 0>]}, ZCurve {keys [<at 0.7 1>,<at 1 0>]}}}
          CurveIndex=&Abs_Float(Float:NiagaraFloat=&MakeFloatFromVector(
            Vector=&ReturnCameraVector(Camera Vectors=::NewEnumerator1, Coordinate Space=::Local)
            Channel=::NewEnumerator0
           ))
          Scale Curve=<1 0 0>
         )
         Front End=&VectorFromCurve(
          VectorCurve=DataInterface() {propertyValues {XCurve {keys [<at 0 0>,<at 0.5 1>]}, YCurve {keys [<at 0.7 1>,<at 1 0>]}, ZCurve {keys [<at 0.7 1>,<at 1 0>]}}}
          CurveIndex=&Abs_Float(Float:NiagaraFloat=&MakeFloatFromVector(
            Vector=&ReturnCameraVector(Camera Vectors=::NewEnumerator1, Coordinate Space=::Local)
            Channel=::NewEnumerator0
           ))
          Scale Curve=<1 0 0>
         )
        )
        Vector B=$Particles.SpriteFacing
       ))
      Input Min=0.2
      Input Max=0.9
      Output Min=0
      Output Max=1
     )
    )
    Particles.Text Size:Vector2f=/Niagara/DynamicInputs/ScaleByCurve/ScaleVector2DByCurve(
     InitialValue:Vector2f=<25 25>
     ScaleCurve:NiagaraDataInterfaceVectorCurve=DataInterface() {propertyValues {
       XCurve {defaultValue=0, keys [<at 0.0893 0.05>, <at 0.2 1>, <at 0.8 1>, {tangentMode=RCTM_User, time=0.9256, value=0.01}]}
       YCurve {defaultValue=0, keys [{tangentMode=RCTM_User, time=0, value=0.1}, {tangentMode=RCTM_User, time=0.0332, value=1}, <at 0.9 1.025>, <at 1 0>]}
       ZCurve {keys []}
      }}
     CurveIndex:NiagaraFloat=$Particles.NormalizedAge
    )
   ) {isSetParameters=true}
   /SetVariables(
    Particles.FinishedTextColor:LinearColor=&MakeLinearColorFromVectorAndFloat(
     "Vector (RGB)"=&MakeVectorFromLinearColorRGB(Color=&ColorFromCurve(ColorCurve=DataInterface() {propertyValues {
         AlphaCurve {defaultValue=1, keys [<at 0 1 RCIM_Constant>]}
         BlueCurve {keys [<at 0.2006 1.5 RCIM_Constant>,<at 0.264 0 RCIM_Constant>]}
         GreenCurve {keys [<at 0.2006 0.5831 RCIM_Constant>,<at 0.264 1.5 RCIM_Constant>]}
         RedCurve {keys [<at 0.2006 0 RCIM_Constant>,<at 0.264 0.2205 RCIM_Constant>]}
        }}))
     "Float (Alpha)"=&Multiply_Float(
      A=&FloatFromCurve(FloatCurve=DataInterface() {propertyValues {Curve {keys [<at 0 1>,<at 0.2768 1 RCIM_Constant>,<at 0.3695 0 RCIM_Constant>,<at 0.4781 1 RCIM_Constant>,<at 0.5837 0 RCIM_Constant>,<at 0.6772 1 RCIM_Constant>,<at 0.7768 0 RCIM_Constant>,<at 0.8613 1 RCIM_Constant>,<at 1 0>]}}})
      B=&RemapRange(
       Input Value To Remap=&Abs_Float(Float:NiagaraFloat=&DotProduct(
         Vector A=&FN_FrontEndVector3Modifier(
          Gameplay=&VectorFromCurve(
           VectorCurve=DataInterface() {propertyValues {XCurve {keys [<at 0 0>,<at 0.2 1>]}, YCurve {keys [<at 0.7 1>,<at 1 0>]}, ZCurve {keys [<at 0.7 1>,<at 1 0>]}}}
           CurveIndex=&Abs_Float(Float:NiagaraFloat=&MakeFloatFromVector(
             Vector=&ReturnCameraVector(Camera Vectors=::NewEnumerator1, Coordinate Space=::Local)
             Channel=::NewEnumerator0
            ))
           Scale Curve=<1 0 0>
          )
          Front End=&VectorFromCurve(
           VectorCurve=DataInterface() {propertyValues {XCurve {keys [<at 0 0>,<at 0.5 1>]}, YCurve {keys [<at 0.7 1>,<at 1 0>]}, ZCurve {keys [<at 0.7 1>,<at 1 0>]}}}
           CurveIndex=&Abs_Float(Float:NiagaraFloat=&MakeFloatFromVector(
             Vector=&ReturnCameraVector(Camera Vectors=::NewEnumerator1, Coordinate Space=::Local)
             Channel=::NewEnumerator0
            ))
           Scale Curve=<1 0 0>
          )
         )
         Vector B=$Particles.SpriteFacing
        ))
       Input Min=0.2
       Input Max=0.9
       Output Min=0
       Output Max=1
      )
     )
    )
    Particles.TypingTextRot=90
    Particles.TypingTextSize:Vector2f=&ScaleVector2DByCurve(
     InitialValue:Vector2f=<28 5.5>
     ScaleCurve:NiagaraDataInterfaceVectorCurve=DataInterface() {propertyValues {
       XCurve {defaultValue=0, keys [{tangentMode=RCTM_User, time=0.9, value=1}, {tangentMode=RCTM_User, time=1, value=0}]}
       YCurve {defaultValue=0, keys [{tangentMode=RCTM_User, time=0.9, value=1}, {tangentMode=RCTM_User, time=1, value=1}]}
       ZCurve {keys []}
      }}
     CurveIndex:NiagaraFloat=$Particles.NormalizedAge
    )
    Particles.TypingTextSubUV:NiagaraFloat=&FloatFromCurve(FloatCurve=DataInterface() {propertyValues {Curve {keys [<at 0 0>,<at 0.05 2 RCIM_Constant>,<at 0.18 3 RCIM_Constant>]}}})
   ) {isSetParameters=true}
   /SetVariables(
    Particles.TypingText2:NiagaraFloat=&FloatFromCurve(FloatCurve=DataInterface() {propertyValues {Curve {keys [<at 0 0>,<at 0.1142 3>]}}})
    Particles.TypingTextScale2:Vector2f=&ScaleVector2DByCurve(
     InitialValue:Vector2f=<23 9.1132>
     ScaleCurve:NiagaraDataInterfaceVectorCurve=DataInterface() {propertyValues {
       XCurve {defaultValue=0, keys [{tangentMode=RCTM_User, time=0.9, value=1}, {tangentMode=RCTM_User, time=1, value=1}]}
       YCurve {defaultValue=0, keys [{tangentMode=RCTM_User, time=0.9496, value=1}, {tangentMode=RCTM_User, time=1, value=0}]}
       ZCurve {keys []}
      }}
     CurveIndex:NiagaraFloat=$Particles.NormalizedAge
    )
   ) {isSetParameters=true}
  ]
  rendererClasses [&NiagaraSpriteRendererProperties]
  renderers [
   {rendererClass=&NiagaraSpriteRendererProperties}
   {rendererClass=&NiagaraSpriteRendererProperties, rendererIndex=1}
   {rendererClass=&NiagaraSpriteRendererProperties, rendererIndex=2}
   {rendererClass=&NiagaraSpriteRendererProperties, rendererIndex=3}
  ]
 }
}
summary {
 systemName=NS_Emote_CarbideWeld_Screen1
 emitters [
  {emitterName=StaticCircle, rendererClasses [&NiagaraSpriteRendererProperties]}
  {emitterName=SmallBarsLWE, rendererClasses []}
  {emitterName=BigBarLWE, rendererClasses []}
  {emitterName=PentagonLWE, rendererClasses []}
  {emitterName=FlashLWE, rendererClasses []}
  {emitterName=LineLWE, rendererClasses []}
  {emitterName=SpinningCircleLWE, rendererClasses []}
 ]
 userVariables [
  User.bIsFrontEnd {type=NiagaraBool}
  User.bIsFrontEndPreview {type=NiagaraBool}
  User.FrontendPos {type=Vector3f, description="", defaultValue {struct=Vector3f, value=<0 0 0>}}
 ]
}
