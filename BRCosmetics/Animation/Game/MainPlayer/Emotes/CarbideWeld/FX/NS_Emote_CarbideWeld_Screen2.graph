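# Niagara system for the CarbideWeld emote that draws animated holographic screen graphics using sprites.
# Line spawns staggered, tagged sprite groups at offsets from the gadget mesh's root socket; SmallBars draws jittering bars that shift orange to green and flicker, with animated lines and underline.
# Both fade alpha by local camera angle, with Line choosing separate front-end and gameplay curves.
# The other six emitters list no renderers.
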NiagaraSystem /BRCosmetics/Animation/Game/MainPlayer/Emotes/CarbideWeld/FX/NS_Emote_CarbideWeld_Screen2
dependencies {
 usedDataInterfaces [
  &NiagaraDataInterfaceArrayFloat3
  &NiagaraDataInterfaceCurve
  &NiagaraDataInterfaceColorCurve
  &NiagaraDataInterfaceVectorCurve
  &NiagaraDataInterfaceCamera
  &NiagaraDataInterfaceVector2DCurve
  &NiagaraDataInterfaceSocketReader
 ]
 usedDynamicInputs []
 usedModules [
  &self:SystemSpawnScript.SystemScriptSource.SystemScriptGraph.NiagaraNodeAssignment_1.SetVariables_#1
  &SystemState
  &self:SystemSpawnScript.SystemScriptSource.SystemScriptGraph.NiagaraNodeAssignment_0.SetVariables_#2
  &self:SystemSpawnScript.SystemScriptSource.SystemScriptGraph.NiagaraNodeAssignment_2.SetVariables_#3
 ]
}
emitters {
 Line {
  emitterUpdateScript [
   &EmitterState(Life Cycle Mode=::NewEnumerator1, Loop Behavior=::NewEnumerator1, Loop Duration=1.6, UseLoopDelay=false)
   &SpawnBurst_Instantaneous(Use Spawn Probability=false, UseLoopCountLimit=false)
   &SpawnBurst_Instantaneous(Spawn Group=1, Spawn Time=0.8, Use Spawn Probability=false, UseLoopCountLimit=false) {module=SpawnBurst_Instantaneous001}
   &SpawnBurst_Instantaneous(Spawn Group=2, Use Spawn Probability=false, UseLoopCountLimit=false) {module=SpawnBurst_Instantaneous002}
   &SpawnBurst_Instantaneous(Spawn Group=3, Spawn Time=0.8, Use Spawn Probability=false, UseLoopCountLimit=false) {module=SpawnBurst_Instantaneous003}
  ]
  particleSpawnScript [
   /SetVariables(Particles.VisibilityTag=&SpawnGroup()) {isSetParameters=true}
   &SocketLocation(
    Position Sampling=::NewEnumerator1
    SamplePosOffset=false
    Sampled Position Offset=<0 0 -710.8568>
    Socket Reader=DataInterface() {propertyValues {
      AttachComponentClass=null
      AttachComponentTag=None
      EditorPreviewAsset=/BRCosmetics/Animation/Props/Gadgets_CarbideWeld/Meshes/CarbideWeld_Gadget
      SourceActorRef=null
      SourceAsset=&CarbideWeld_Gadget
      SourceMode=Default
      bRequireCurrentFrameData=true
      bUpdateSocketsPerFrame=true
      FilteredSockets [rootSocket]
      ObjectParameterBinding {parameter {name=None, typeDefHandle {registeredTypeIndex=1}, varData []}}
     }}
    Socket Sampling Mode=::NewEnumerator3
    VelocitySpeedLimit=false
   )
   &PartitionParticles(Number Of Partitions=4)
   &InitializeParticle(
    Lifetime=1.5
    Position Offset=&SelectVectorFromArray(
     Direct Array Index=$Particles.PartitionParticles.Partition
     Vector Selection Array=DataInterface() {propertyValues {FloatData [<0 -20 -70>, <0 -20 -67>, <0 20 -42>, <0 20 -38>]}}
    )
    Position Offset Coordinate Space=::Local
    Sprite Rotation Angle=90
    Sprite Rotation Mode=::NewEnumerator0
    Sprite Size Mode=::NewEnumerator3
    Uniform Sprite Size=53
    UsePositionOffset=true
    Write Material Random=true
   )
   &SpriteFacingAndAlignment(Sprite Alignment=<0 -1 0>, Sprite Facing=<1 0.6 0>, WriteAlignment=true, WriteFacing=true)
   /SetVariables(Particles.Alignment2:Vector3f=<0 1 0>, Particles.Facing2:Vector3f=<-1 -0.6 0>) {isSetParameters=true}
   &InheritVelocity(Speed Limit=false, UseSpeedThreshold=false) {enabled=false}
   /SetVariables(Particles.Alignment3:Vector3f=<0 -1 0>, Particles.Facing3:Vector3f=<1 -1 0>) {isSetParameters=true}
  ]
  particleUpdateScript [
   &ParticleState()
   &ScaleSpriteSize(
    Non-Uniform Curve Sprite Scale=DataInterface() {propertyValues {
      XCurve {keys [{tangentMode=RCTM_User, time=0, value=0}, {tangentMode=RCTM_User, time=0.0731, value=1}, <at 0.9418 1>, <at 1 0>]}
      YCurve {keys [{tangentMode=RCTM_User, time=0, value=1}]}
     }}
    Scale Sprite Size Mode=::NewEnumerator3
   )
   &SolveForcesAndVelocity(Clamp Velocity=false, Limit Acceleration=false)
   /SetVariables(
    Particles.RotationText=180
    Particles.TextColor:LinearColor=&MakeLinearColorFromVectorAndFloat(
     "Vector (RGB)"=&MakeVectorFromLinearColorRGB(Color=<rgba 0 0.3898 1.2 1>)
     "Float (Alpha)"=&RemapRange(
      Input Value To Remap=&Abs_Float(Float:NiagaraFloat=&DotProduct(
        Vector A=&FN_FrontEndVector3Modifier(
         Gameplay=&VectorFromCurve(
          VectorCurve=DataInterface() {propertyValues {XCurve {keys [<at 0 0>,<at 0.2 1>]}, YCurve {keys [<at 0.7 1>,<at 1 0>]}, ZCurve {keys [<at 0.7 1>,<at 1 0>]}}}
          CurveIndex=&Abs_Float(Float:NiagaraFloat=&MakeFloatFromVector(
            Vector=&ReturnCameraVector(Camera Vectors=::NewEnumerator1, Coordinate Space=::Local)
            Channel=::NewEnumerator0
           ))
          Scale Curve=<1 0 0>
         )
         Front End=&VectorFromCurve(
          VectorCurve=DataInterface() {propertyValues {XCurve {keys [<at 0 0>,<at 0.5 1>]}, YCurve {keys [<at 0.7 1>,<at 1 0>]}, ZCurve {keys [<at 0.7 1>,<at 1 0>]}}}
          CurveIndex=&Abs_Float(Float:NiagaraFloat=&MakeFloatFromVector(
            Vector=&ReturnCameraVector(Camera Vectors=::NewEnumerator1, Coordinate Space=::Local)
            Channel=::NewEnumerator0
           ))
          Scale Curve=<1 0 0>
         )
        )
        Vector B=<1 0 0>
       ))
      Input Min=0.2
      Input Max=0.9
      Output Min=0
      Output Max=1
     )
    )
    Particles.TextSize:Vector2f=/Niagara/DynamicInputs/ScaleByCurve/ScaleVector2DByCurve(
     InitialValue:Vector2f=<14.6124 32.8192>
     ScaleCurve:NiagaraDataInterfaceVectorCurve=DataInterface() {propertyValues {
       XCurve {defaultValue=1, keys [{tangentMode=RCTM_User, time=0, value=0}, <at 0.05 1>, <at 0.93 1>, {tangentMode=RCTM_User, time=1, value=0}]}
       YCurve {defaultValue=1, keys [{tangentMode=RCTM_User, time=0, value=1}, {tangentMode=RCTM_User, time=0.2937, value=1}, <at 0.75 1>, <at 1 1>]}
       ZCurve {keys []}
      }}
     CurveIndex:NiagaraFloat=$Particles.NormalizedAge
    )
    Particles.float:NiagaraFloat=&FloatFromCurve(FloatCurve=DataInterface() {propertyValues {Curve {keys [<at 0 0>,<at 0.8 3>]}}})
   ) {isSetParameters=true}
   /SetVariables(Particles.LineColor:LinearColor=&MakeLinearColorFromVectorAndFloat(
     "Vector (RGB)"=&MakeVectorFromLinearColorRGB(Color=<rgba 2 0.3811 0.0324 1>)
     "Float (Alpha)"=&Multiply_Float(
      A=&FloatFromCurve(FloatCurve=DataInterface() {propertyValues {Curve {keys [<at 0.2 1>,<at 2.8503 1>]}}}, CurveIndex=$Engine.System.Age)
      B=&RemapRange(
       Input Value To Remap=&Abs_Float(Float:NiagaraFloat=&DotProduct(
         Vector A=&FN_FrontEndVector3Modifier(
          Gameplay=&VectorFromCurve(
           VectorCurve=DataInterface() {propertyValues {XCurve {keys [<at 0 0>,<at 0.2 1>]}, YCurve {keys [<at 0.7 1>,<at 1 0>]}, ZCurve {keys [<at 0.7 1>,<at 1 0>]}}}
           CurveIndex=&Abs_Float(Float:NiagaraFloat=&MakeFloatFromVector(
             Vector=&ReturnCameraVector(Camera Vectors=::NewEnumerator1, Coordinate Space=::Local)
             Channel=::NewEnumerator0
            ))
           Scale Curve=<1 0 0>
          )
          Front End=&VectorFromCurve(
           VectorCurve=DataInterface() {propertyValues {XCurve {keys [<at 0 0>,<at 0.5 1>]}, YCurve {keys [<at 0.7 1>,<at 1 0>]}, ZCurve {keys [<at 0.7 1>,<at 1 0>]}}}
           CurveIndex=&Abs_Float(Float:NiagaraFloat=&MakeFloatFromVector(
             Vector=&ReturnCameraVector(Camera Vectors=::NewEnumerator1, Coordinate Space=::Local)
             Channel=::NewEnumerator0
            ))
           Scale Curve=<1 0 0>
          )
         )
         Vector B=<1 0 0>
        ))
       Input Min=0.2
       Input Max=0.9
       Output Min=0
       Output Max=1
      )
     )
    )) {isSetParameters=true}
  ]
  rendererClasses [&NiagaraSpriteRendererProperties]
  renderers [
   {rendererClass=&NiagaraSpriteRendererProperties}
   {rendererClass=&NiagaraSpriteRendererProperties, rendererIndex=1}
   {rendererClass=&NiagaraSpriteRendererProperties, rendererIndex=2}
   {rendererClass=&NiagaraSpriteRendererProperties, rendererIndex=3}
  ]
 }
 PentagonLWE {rendererClasses [], renderers []}
 PopUpText {rendererClasses [], renderers []}
 PopUpText001 {rendererClasses [], renderers []}
 SmallBars {
  emitterUpdateScript [
   &EmitterState(Life Cycle Mode=::NewEnumerator1, Loop Behavior=::NewEnumerator1, Loop Delay=2.6, Loop Duration=2.9, UseLoopDelay=true)
   &SpawnBurst_Instantaneous(Spawn Count=5, Spawn Time=0.2, Use Spawn Probability=false, UseLoopCountLimit=false) {module=SpawnBurst_Instantaneous002}
   &SpawnBurst_Instantaneous(Spawn Count=6, Spawn Time=0.244, Use Spawn Probability=false, UseLoopCountLimit=false) {enabled=false}
   &SpawnBurst_Instantaneous(Spawn Group=1, Use Spawn Probability=false, UseLoopCountLimit=false) {module=SpawnBurst_Instantaneous001}
  ]
  particleSpawnScript [
   &PartitionParticles(
    Partition Count Mode:ENiagara_InfiniteLoopDuration=::NewEnumerator1
    Partition Mode:ENiagara_PartitionMode_v2=::NewEnumerator1
    Write Partition to Ribbon ID=::NewEnumerator0
   )
   &InitializeParticle(
    Lifetime=2.8
    Mesh Scale=<0.02 0.02 0.04>
    Mesh Scale Mode=::NewEnumerator4
    Position Mode=::NewEnumerator1
    Position Offset=&SelectVectorFromArray(
     Direct Array Index=&ReturnParticleIDAttributes()
     Vector Selection Array=DataInterface() {propertyValues {FloatData [<0 -19.5 -2.6962>, <0 -21 -2.6962>, <0 -22.5 -2.6962>, <0 -24 -2.6962>, <0 -25.5 -2.6962>, <0 -19.5 -2.6962>]}}
    )
    Sprite Rotation Angle=180
    Sprite Rotation Mode=::NewEnumerator0
    Sprite Size Mode=::NewEnumerator3
    Uniform Sprite Size=16.8642
    UsePositionOffset=true
    Write Material Random=true
   )
   &SpriteFacingAndAlignment(WriteAlignment=true, WriteFacing=true)
   /SetVariables(
    Particles.Line Position:NiagaraPosition=&ConvertPositionToVector(Input Position=<0 -18.4168 -2.6962>, Conversion Operation:ENiagara_LWCConvertPosToVec=::NewEnumerator0)
    Particles.Line Position 2:NiagaraPosition=&ConvertPositionToVector(Input Position=<0 -26.5 -2.6962>, Conversion Operation:ENiagara_LWCConvertPosToVec=::NewEnumerator0)
    Particles.Line2Facing:Vector3f=<-1 0 0>
   ) {isSetParameters=true}
   /SetVariables() {isSetParameters=true}
   /SetVariables(Particles.Underline Pos:NiagaraPosition=&ConvertPositionToVector(Input Position=<0 -17.5 -3.5624>, Conversion Operation:ENiagara_LWCConvertPosToVec=::NewEnumerator0)) {isSetParameters=true}
   /SetVariables(Particles.VisibilityTag=&SpawnGroup()) {isSetParameters=true}
  ]
  particleUpdateScript [
   &ParticleState()
   /SetVariables(
    Particles.Bar Size:Vector2f=&MakeVector2D(
     X=&Add_Float(
      A=&FloatFromCurve(
       FloatCurve=DataInterface() {propertyValues {Curve {keys [
           {interpMode=RCIM_Cubic, leaveTangent=7.3797, tangentMode=RCTM_Break, time=0, value=0}
           {arriveTangent=2.8308, tangentMode=RCTM_User, time=0.25, value=1}
           {arriveTangent=-4.3555, leaveTangent=-19.5761, time=0.74, value=1}
           {arriveTangent=-0.5344, interpMode=RCIM_Cubic, tangentMode=RCTM_Break, time=0.78, value=-0.1}
          ]}}}
       Scale Curve=10
      )
      B=&Sine(
       Normalized Angle=$Emitter.Age
       Period=&RandomRangeFloat(Minimum=0.5, Maximum=2)
       Scale=&FloatFromCurve(
        FloatCurve=DataInterface() {propertyValues {Curve {keys [<at 0 1>,<at 0.3 1>,<at 0.42 0>]}}}
        Scale Curve=&RandomRangeFloat(Minimum=0.4, Maximum=3)
       )
       Bias=1
      )
     )
     Y=2
    )
    Particles.BarColor:LinearColor=&MakeLinearColorFromVectorAndFloat(
     "Vector (RGB)"=&MakeVectorFromLinearColorRGB(Color=&ColorFromCurve(ColorCurve=DataInterface() {propertyValues {
         AlphaCurve {keys [<at 0.15 1 RCIM_Constant>,<at 0.3 1 RCIM_Constant>,<at 0.4675 0 RCIM_Constant>,<at 0.52 1 RCIM_Constant>,<at 0.5741 0 RCIM_Constant>,<at 0.6412 1 RCIM_Constant>,<at 0.7 0 RCIM_Constant>,<at 0.7447 1 RCIM_Constant>,<at 0.8141 0 RCIM_Constant>,<at 0.8683 1 RCIM_Constant>,<at 0.9277 0 RCIM_Constant>,<at 0.9897 0>]}
         BlueCurve {keys [<at 0.392 0.0324 RCIM_Constant>,<at 0.4812 0 RCIM_Constant>]}
         GreenCurve {keys [<at 0.392 0.375 RCIM_Constant>,<at 0.4812 2 RCIM_Constant>]}
         RedCurve {keys [<at 0.392 2 RCIM_Constant>,<at 0.4812 0.1416 RCIM_Constant>]}
        }}))
     "Float (Alpha)"=&Multiply_Float(
      A=&FloatFromCurve(FloatCurve=DataInterface() {propertyValues {Curve {keys [<at 0 1>,<at 0.15 1 RCIM_Constant>,<at 0.3 1 RCIM_Constant>,<at 0.4675 0 RCIM_Constant>,<at 0.52 1 RCIM_Constant>,<at 0.5741 0 RCIM_Constant>,<at 0.6412 1 RCIM_Constant>,<at 0.7 0 RCIM_Constant>,<at 0.7447 1 RCIM_Constant>,<at 0.8141 0 RCIM_Constant>,<at 0.8683 1 RCIM_Constant>,<at 0.9277 0 RCIM_Constant>,<at 0.9897 0>,<at 1 0>]}}})
      B=&RemapRange(
       Input Value To Remap=&Abs_Float(Float:NiagaraFloat=&DotProduct(
         Vector A=&VectorFromCurve(
          VectorCurve=DataInterface() {propertyValues {XCurve {keys [<at 0 0>,<at 0.5 1>]}, YCurve {keys [<at 0.5 1>,<at 1 0>]}, ZCurve {keys [<at 0.5 1>,<at 1 0>]}}}
          CurveIndex=&Abs_Float(Float:NiagaraFloat=&MakeFloatFromVector(
            Vector=&ReturnCameraVector(Camera Vectors=::NewEnumerator1, Coordinate Space=::Local)
            Channel=::NewEnumerator0
           ))
          Scale Curve=<1 0 0>
         )
         Vector B=<1 0 0>
        ))
       Input Min=0.2
       Input Max=0.9
       Output Min=0
       Output Max=1
      )
     )
    )
   ) {isSetParameters=true}
   &ScaleSpriteSize(
    Non-Uniform Curve Sprite Scale=DataInterface() {propertyValues {
      XCurve {keys [{tangentMode=RCTM_User, time=0.04, value=0.1}, {tangentMode=RCTM_User, time=0.1476, value=1}, <at 0.9 1>, <at 1 0>]}
      YCurve {keys [{tangentMode=RCTM_User, time=0, value=1}, {tangentMode=RCTM_User, time=1, value=1}]}
     }}
    Scale Sprite Size Mode=::NewEnumerator3
   )
   &SolveForcesAndVelocity(Clamp Velocity=false, Limit Acceleration=false)
   /SetVariables(Particles.Line Position 2:NiagaraPosition=&Lerp_Vector(
     A=<0 -18.7606 -2.6962>
     B=$Particles.Initial.Line Position 2
     Alpha=&FloatFromCurve(FloatCurve=DataInterface() {propertyValues {Curve {
         defaultValue=0
         keys [
          {arriveTangent=10.5637, interpMode=RCIM_Cubic, leaveTangent=10.5637, tangentMode=RCTM_User, time=0, value=0}
          {arriveTangent=0.8002, leaveTangent=0.8002, time=0.06, value=1}
          <at 0.9 1>
          {arriveTangent=0.0224, interpMode=RCIM_Cubic, leaveTangent=0.0224, tangentMode=RCTM_User, time=1, value=0}
         ]
        }}})
    )) {isSetParameters=true}
   /SetVariables(Particles.LineColor:LinearColor=$System.LineColor) {isSetParameters=true}
   /SetVariables(
    Particles.Underline Color:LinearColor=$System.LineColor
    Particles.UnderlineSize:Vector2f=&ScaleVector2DByCurve(
     InitialValue:Vector2f=<1 20>
     ScaleCurve:NiagaraDataInterfaceVectorCurve=DataInterface() {propertyValues {
       XCurve {defaultValue=0, keys [{tangentMode=RCTM_User, time=0, value=1}, {tangentMode=RCTM_User, time=1, value=1}]}
       YCurve {defaultValue=0, keys [{tangentMode=RCTM_User, time=0, value=0}, {tangentMode=RCTM_User, time=0.15, value=1}, <at 0.9 1>, <at 1 0>]}
       ZCurve {keys []}
      }}
     CurveIndex:NiagaraFloat=$Particles.NormalizedAge
    )
   ) {isSetParameters=true}
  ]
  rendererClasses [&NiagaraSpriteRendererProperties]
  renderers [
   {rendererClass=&NiagaraSpriteRendererProperties}
   {rendererClass=&NiagaraSpriteRendererProperties, rendererIndex=1}
   {rendererClass=&NiagaraSpriteRendererProperties, rendererIndex=2}
   {rendererClass=&NiagaraSpriteRendererProperties, rendererIndex=3}
  ]
 }
 SmallBarsLWE003 {rendererClasses [], renderers []}
 SpinningCircleLWE {rendererClasses [], renderers []}
 StaticCircle002 {rendererClasses [], renderers []}
}
summary {
 systemName=NS_Emote_CarbideWeld_Screen2
 emitters [
  {emitterName=SmallBars, rendererClasses [&NiagaraSpriteRendererProperties]}
  {emitterName=SpinningCircleLWE, rendererClasses []}
  {emitterName=PentagonLWE, rendererClasses []}
  {emitterName=SmallBarsLWE003, rendererClasses []}
  {emitterName=PopUpText, rendererClasses []}
  {emitterName=StaticCircle002, rendererClasses []}
  {emitterName=PopUpText001, rendererClasses []}
  {emitterName=Line, rendererClasses [&NiagaraSpriteRendererProperties]}
 ]
 userVariables [
  User.bIsFrontEnd {type=NiagaraBool}
  User.bIsFrontEndPreview {type=NiagaraBool}
  User.NewVector3f {type=Vector3f, description="", defaultValue {struct=Vector3f, value=<0 0 0>}}
  User.NormalizedNotifyProgress {type=NiagaraFloat, defaultValue=0, description=""}
 ]
}
