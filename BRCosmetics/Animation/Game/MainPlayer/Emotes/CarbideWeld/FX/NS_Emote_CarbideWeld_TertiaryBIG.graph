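# Niagara effect system for the CarbideWeld player emote, named as its big tertiary variant.
# It groups nine DirectionalBurst emitters with one DecalLWE emitter, whose names suggest directional bursts plus a decal.
# SystemState is the only listed module, and the system exposes no user variables.
# No renderers, data interfaces or dynamic inputs are listed, so the visual output of each emitter is not shown.
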NiagaraSystem /BRCosmetics/Animation/Game/MainPlayer/Emotes/CarbideWeld/FX/NS_Emote_CarbideWeld_TertiaryBIG
dependencies {usedDataInterfaces [], usedDynamicInputs [], usedModules [&SystemState]}
emitters {
 DecalLWE {rendererClasses [], renderers []}
 DirectionalBurst {rendererClasses [], renderers []}
 DirectionalBurst001 {rendererClasses [], renderers []}
 DirectionalBurst002 {rendererClasses [], renderers []}
 DirectionalBurst003 {rendererClasses [], renderers []}
 DirectionalBurst004 {rendererClasses [], renderers []}
 DirectionalBurst005 {rendererClasses [], renderers []}
 DirectionalBurst006 {rendererClasses [], renderers []}
 DirectionalBurst007 {rendererClasses [], renderers []}
 DirectionalBurst008 {rendererClasses [], renderers []}
}
summary {
 systemName=NS_Emote_CarbideWeld_TertiaryBIG
 emitters [
  {emitterName=DirectionalBurst, rendererClasses []}
  {emitterName=DirectionalBurst001, rendererClasses []}
  {emitterName=DirectionalBurst002, rendererClasses []}
  {emitterName=DirectionalBurst003, rendererClasses []}
  {emitterName=DirectionalBurst004, rendererClasses []}
  {emitterName=DecalLWE, rendererClasses []}
  {emitterName=DirectionalBurst006, rendererClasses []}
  {emitterName=DirectionalBurst007, rendererClasses []}
  {emitterName=DirectionalBurst005, rendererClasses []}
  {emitterName=DirectionalBurst008, rendererClasses []}
 ]
 userVariables []
}
