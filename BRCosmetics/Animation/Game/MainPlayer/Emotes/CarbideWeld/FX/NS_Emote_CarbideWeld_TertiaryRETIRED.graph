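# Retired tertiary Niagara effect for the CarbideWeld main-player emote, marked RETIRED in its name and left inert.
# It holds two DirectionalBurst emitters, both disabled and listing no renderers, so it draws nothing.
# SystemState is the only module it uses, with no user variables, data interfaces or dynamic inputs.
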NiagaraSystem /BRCosmetics/Animation/Game/MainPlayer/Emotes/CarbideWeld/FX/NS_Emote_CarbideWeld_TertiaryRETIRED
dependencies {usedDataInterfaces [], usedDynamicInputs [], usedModules [&SystemState]}
emitters {DirectionalBurst {bEnabled=false, rendererClasses [], renderers []}, DirectionalBurst001 {bEnabled=false, rendererClasses [], renderers []}}
summary {
 systemName=NS_Emote_CarbideWeld_TertiaryRETIRED
 emitters [{bEnabled=false, emitterName=DirectionalBurst, rendererClasses []}, {bEnabled=false, emitterName=DirectionalBurst001, rendererClasses []}]
 userVariables []
}
